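# Sound cue for NitroDrome fireworks bursts in their 2D variant, playing one of fifteen burst wave variations.
# An equally weighted Random node, excluded from branch culling, picks the wave and feeds an Attenuation node at the root.
# The cue is routed to the Fort_Traps sound class, and both the cue and its Attenuation node reference the NitroDrome fireworks 2D attenuation settings.
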SoundCue /NitroDrome/Assets/Audio/Fireworks/NitroDrome_Fireworks_Burst_2D_Cue
root=Attenuation_1
nodes [
 {id=Attenuation_1, children [Random_1], properties {attenuationSettings=&/NitroDrome_Fireworks_2D_Att}}
 {
  id=Random_1
  children [
   WavePlayer_11
   WavePlayer_12
   WavePlayer_13
   WavePlayer_14
   WavePlayer_15
   WavePlayer_16
   WavePlayer_17
   WavePlayer_18
   WavePlayer_19
   WavePlayer_20
   WavePlayer_21
   WavePlayer_22
   WavePlayer_23
   WavePlayer_24
   WavePlayer_25
  ]
  properties {bShouldExcludeFromBranchCulling=true, weights [1,1,1,1,1,1,1,1,1,1,1,1,1,1,1]}
 }
 {id=WavePlayer_11, properties {soundWaveAssetPtr=&/NitroDrome_Fireworks_Burst_2D_01}}
 {id=WavePlayer_12, properties {soundWaveAssetPtr=&/NitroDrome_Fireworks_Burst_2D_02}}
 {id=WavePlayer_13, properties {soundWaveAssetPtr=&/NitroDrome_Fireworks_Burst_2D_03}}
 {id=WavePlayer_14, properties {soundWaveAssetPtr=&/NitroDrome_Fireworks_Burst_2D_04}}
 {id=WavePlayer_15, properties {soundWaveAssetPtr=&/NitroDrome_Fireworks_Burst_2D_05}}
 {id=WavePlayer_16, properties {soundWaveAssetPtr=&/NitroDrome_Fireworks_Burst_2D_06}}
 {id=WavePlayer_17, properties {soundWaveAssetPtr=&/NitroDrome_Fireworks_Burst_2D_07}}
 {id=WavePlayer_18, properties {soundWaveAssetPtr=&/NitroDrome_Fireworks_Burst_2D_08}}
 {id=WavePlayer_19, properties {soundWaveAssetPtr=&/NitroDrome_Fireworks_Burst_2D_09}}
 {id=WavePlayer_20, properties {soundWaveAssetPtr=&/NitroDrome_Fireworks_Burst_2D_10}}
 {id=WavePlayer_21, properties {soundWaveAssetPtr=&/NitroDrome_Fireworks_Burst_2D_11}}
 {id=WavePlayer_22, properties {soundWaveAssetPtr=&/NitroDrome_Fireworks_Burst_2D_12}}
 {id=WavePlayer_23, properties {soundWaveAssetPtr=&/NitroDrome_Fireworks_Burst_2D_13}}
 {id=WavePlayer_24, properties {soundWaveAssetPtr=&/NitroDrome_Fireworks_Burst_2D_14}}
 {id=WavePlayer_25, properties {soundWaveAssetPtr=&/NitroDrome_Fireworks_Burst_2D_15}}
]
properties {
 attenuationSettings=&NitroDrome_Fireworks_2D_Att
 maxDistance=20000
 soundClassObject=&Fort_Traps
 volumeMultiplier=1
}
