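# Sound cue for NitroDrome firework bursts, layering a positional 3D burst with a quieter 2D burst layer, routed to the Fort_Traps sound class.
# A Mixer blends a random 3D wave and a reduced-volume random 2D wave, each branch with its own attenuation asset.
# A Quality Level node gates the 2D layer, feeding it only in the first two of four slots and leaving the others empty.
# Both Random nodes use equal weights and are excluded from branch culling.
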SoundCue /NitroDrome/Assets/Audio/Fireworks/NitroDrome_Fireworks_Burst_3D_Cue
root=Mixer_0
nodes [
 {id=Attenuation_0, children [Random_0], properties {attenuationSettings=&/NitroDrome_Fireworks_3D_Att}}
 {id=Attenuation_1, children [Random_1], properties {attenuationSettings=&/NitroDrome_Fireworks_2D_Att}}
 {id=Mixer_0, children [Attenuation_0, QualityLevel_0], properties {inputVolume [1,0.25]}}
 {id=QualityLevel_0, children [Attenuation_1, Attenuation_1, null, null]}
 {
  id=Random_0
  children [
   WavePlayer_0
   WavePlayer_1
   WavePlayer_2
   WavePlayer_3
   WavePlayer_4
   WavePlayer_5
   WavePlayer_6
   WavePlayer_7
   WavePlayer_8
   WavePlayer_9
   WavePlayer_10
  ]
  properties {bShouldExcludeFromBranchCulling=true, weights [1,1,1,1,1,1,1,1,1,1,1]}
 }
 {
  id=Random_1
  children [
   WavePlayer_11
   WavePlayer_12
   WavePlayer_13
   WavePlayer_14
   WavePlayer_15
   WavePlayer_16
   WavePlayer_17
   WavePlayer_18
   WavePlayer_19
   WavePlayer_20
   WavePlayer_21
   WavePlayer_22
   WavePlayer_23
   WavePlayer_24
   WavePlayer_25
  ]
  properties {bShouldExcludeFromBranchCulling=true, weights [1,1,1,1,1,1,1,1,1,1,1,1,1,1,1]}
 }
 {id=WavePlayer_0, properties {soundWaveAssetPtr=&/NitroDrome_Fireworks_Burst_3D_11}}
 {id=WavePlayer_1, properties {soundWaveAssetPtr=&/NitroDrome_Fireworks_Burst_3D_10}}
 {id=WavePlayer_10, properties {soundWaveAssetPtr=&/NitroDrome_Fireworks_Burst_3D_01}}
 {id=WavePlayer_11, properties {soundWaveAssetPtr=&/NitroDrome_Fireworks_Burst_2D_01}}
 {id=WavePlayer_12, properties {soundWaveAssetPtr=&/NitroDrome_Fireworks_Burst_2D_02}}
 {id=WavePlayer_13, properties {soundWaveAssetPtr=&/NitroDrome_Fireworks_Burst_2D_03}}
 {id=WavePlayer_14, properties {soundWaveAssetPtr=&/NitroDrome_Fireworks_Burst_2D_04}}
 {id=WavePlayer_15, properties {soundWaveAssetPtr=&/NitroDrome_Fireworks_Burst_2D_05}}
 {id=WavePlayer_16, properties {soundWaveAssetPtr=&/NitroDrome_Fireworks_Burst_2D_06}}
 {id=WavePlayer_17, properties {soundWaveAssetPtr=&/NitroDrome_Fireworks_Burst_2D_07}}
 {id=WavePlayer_18, properties {soundWaveAssetPtr=&/NitroDrome_Fireworks_Burst_2D_08}}
 {id=WavePlayer_19, properties {soundWaveAssetPtr=&/NitroDrome_Fireworks_Burst_2D_09}}
 {id=WavePlayer_2, properties {soundWaveAssetPtr=&/NitroDrome_Fireworks_Burst_3D_09}}
 {id=WavePlayer_20, properties {soundWaveAssetPtr=&/NitroDrome_Fireworks_Burst_2D_10}}
 {id=WavePlayer_21, properties {soundWaveAssetPtr=&/NitroDrome_Fireworks_Burst_2D_11}}
 {id=WavePlayer_22, properties {soundWaveAssetPtr=&/NitroDrome_Fireworks_Burst_2D_12}}
 {id=WavePlayer_23, properties {soundWaveAssetPtr=&/NitroDrome_Fireworks_Burst_2D_13}}
 {id=WavePlayer_24, properties {soundWaveAssetPtr=&/NitroDrome_Fireworks_Burst_2D_14}}
 {id=WavePlayer_25, properties {soundWaveAssetPtr=&/NitroDrome_Fireworks_Burst_2D_15}}
 {id=WavePlayer_3, properties {soundWaveAssetPtr=&/NitroDrome_Fireworks_Burst_3D_08}}
 {id=WavePlayer_4, properties {soundWaveAssetPtr=&/NitroDrome_Fireworks_Burst_3D_07}}
 {id=WavePlayer_5, properties {soundWaveAssetPtr=&/NitroDrome_Fireworks_Burst_3D_06}}
 {id=WavePlayer_6, properties {soundWaveAssetPtr=&/NitroDrome_Fireworks_Burst_3D_05}}
 {id=WavePlayer_7, properties {soundWaveAssetPtr=&/NitroDrome_Fireworks_Burst_3D_04}}
 {id=WavePlayer_8, properties {soundWaveAssetPtr=&/NitroDrome_Fireworks_Burst_3D_03}}
 {id=WavePlayer_9, properties {soundWaveAssetPtr=&/NitroDrome_Fireworks_Burst_3D_02}}
]
properties {maxDistance=20000, soundClassObject=&Fort_Traps, volumeMultiplier=1}
